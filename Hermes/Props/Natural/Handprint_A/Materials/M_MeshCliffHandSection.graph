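# Material for a cliff mesh section that carries a handprint marking, composited over the Hermes cliff layer by texture alpha.
# On desktop with virtual textures, the cliff blends with landscape RVT sampled at an instance-oriented offset, with an edge mask derived from RVT height.
# Mobile uses the layer's mobile blend with derived normals.
# Handprint UVs are scaled, offset and clamped, and its normal blends angle-corrected with the cliff's before output in world space.
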Material /Hermes/Props/Natural/Handprint_A/Materials/M_MeshCliffHandSection
graph {
 properties {
  bAutomaticallySetUsageInEditor=false
  bTangentSpaceNormal=false
  bUseMaterialAttributes=true
  bUsedWithInstancedStaticMeshes=true
  usedShadingModels=Default Lit
 }
 nodes [
  n1 = MaterialFunctionCall!(
   Material
   MobileMaterial=MakeMaterialAttributes!(
    BaseColor
    Metallic
    Specular=Constant()
    Roughness=Constant() {r=1}
    Anisotropy
    EmissiveColor
    Opacity
    OpacityMask
    Normal
    Tangent
    WorldPositionOffset
    SubsurfaceColor
    ClearCoat
    ClearCoatRoughness
    AmbientOcclusion
    Refraction
    CustomizedUVs_0
    CustomizedUVs_1
    CustomizedUVs_2
    CustomizedUVs_3
    CustomizedUVs_4
    CustomizedUVs_5
    CustomizedUVs_6
    CustomizedUVs_7
    PixelDepthOffset
    ShadingModel
    Displacement
   )
   LayerWeight=Constant() {r=1}
   UseTangentNormals=StaticBool()
   TerrainMasks_QuadA=Constant4Vector() {constant=<rgba 1 0 0 0>}.RGBA
   TerrainMasks_QuadB
   TerrainMasks_QuadC
   TerrainMasks_QuadD
   VariationMasks Mobile=MaterialFunctionCall() {materialFunction=/BRCoreMaterials/Hermes/Terrain/Landscape/Materials/MaterialFunctions/MF_Hermes_VariationMasks_Mobile}.VariationMasks Mobile
   MossDesaturateMask
  ) {materialFunction=/BRCoreMaterials/Hermes/Terrain/Landscape/Materials/Layers/MF_Hermes_Cliff}
  n2 = Normalize(VectorInput=Transform(None=Constant3Vector() {constant=<rgba 0 -1 0 0>}) {transformSourceType=TRANSFORMSOURCE_Instance})
  n3 = Multiply(A=ScalarParameter() {defaultValue=150, parameterName=Offset RVT Sample}, B=NamedRerouteUsage())
  n4 = RuntimeVirtualTextureSampleParameter(World Position=Subtract(A=WorldPosition().XYZ, B=Multiply(A=@n2, B=@n3))) {
   materialType=BaseColor_Normal_Specular_YCoCg
   parameterName=SampleCliff
   virtualTexture=/Game/Athena/Apollo/Maps/Landscape/VT/RVT_Apollo_Landscape
  }
  n5 = MakeMaterialAttributes(BaseColor=@n4.BaseColor, Specular=@n4.BaseColor, Roughness=@n4.BaseColor, Normal=@n4.BaseColor)
  n6 = NamedRerouteUsage()
  n7 = ShadingPathSwitch(
   Default=VirtualTextureFeatureSwitch!(No=@n1.CliffMaterialOnly, Yes=BlendMaterialAttributes(A=@n5, B=@n1.CliffMaterialOnly, Alpha=@n6))
   Mobile=MaterialFunctionCall!(MobileMaterial=@n1.MobileBlendedMaterial, TransformTangents=StaticBool()) {materialFunction=/Game/Athena/Helios/Environments/Landscape/Materials/MaterialFunctions_Utilities/MF_MobileDeriveNormals}.ResultMaterial
  )
  n8 = ScalarParameter() {defaultValue=0.8, parameterName=UVScale}
  n9 = ScalarParameter() {parameterName=UVOffsetX}
  n10 = ScalarParameter() {parameterName=UVOffsetY}
  n11 = Saturate(None=Subtract(
    A=MaterialFunctionCall(Texture Scale=@n8) {materialFunction=&ScaleUVsByCenter}.UVs
    B=MaterialFunctionCall(X=@n9, Y=@n10) {materialFunction=&MakeFloat2}.Result
   ))
  n12 = TextureSample(UVs=@n11) {texture=/Hermes/Props/Natural/Handprint_A/Textures/T__Handprint_A_A_D}
  n13 = TextureSample(UVs=@n11) {samplerType=SAMPLERTYPE_Masks, texture=/Hermes/Props/Natural/Handprint_A/Textures/T_Handprint_A_A_S}
  n14 = MaterialFunctionCall(
   Normal=TextureSample(UVs=@n11) {samplerType=SAMPLERTYPE_Normal, texture=/Hermes/Props/Natural/Handprint_A/Textures/T_Handprint_A_A_N}.RGB
   Flatness=ScalarParameter() {defaultValue=0.2, parameterName=Hand Normal Strength}
  ) {materialFunction=&FlattenNormal}
  n15 = Transform(None=NamedRerouteUsage()) {transformSourceType=TRANSFORMSOURCE_World, transformType=TRANSFORM_Tangent}
  n16 = Normalize(VectorInput=LinearInterpolate(
    A=@n14.Result
    B=MaterialFunctionCall(BaseNormal=@n15, AdditionalNormal=@n14.Result) {materialFunction=&BlendAngleCorrectedNormals}.Result
    Alpha=ScalarParameter() {defaultValue=1, parameterName=CliffNormalBlend}
   ))
  BlendMaterialAttributes(A=@n7, B=MakeMaterialAttributes(BaseColor=@n12.RGB, Specular=@n13.R, Roughness=@n13.R, Normal=Transform(None=@n16)), Alpha=@n12.A)
  n17 = RuntimeVirtualTextureSampleParameter() {materialType=WorldHeight, parameterName=SampleCliffHeight, virtualTexture=/Game/Athena/Apollo/Maps/Landscape/VT/RVT_Apollo_Landscape_Height}
  n18 = Subtract(A=WorldPosition().Z, B=@n17.WorldHeight)
  n19 = Divide(A=@n18, B=ScalarParameter() {defaultValue=938.4503, parameterName=HeightBlendDistance})
  NamedRerouteDeclaration(None=Saturate(None=Multiply(A=@n19) {constB=3, desc=clip to 1})) {name=RVT Edge Mask, nodeColor=<rgba 1 0 0.8437 1>}
  NamedRerouteDeclaration(None=GetMaterialAttributes(None=@n7) {attributeGetTypes [#1]}.Normal) {name=CliffNormal, nodeColor=<rgba 1 0 0.2344 1>}
  Saturate(None=@n18)
  SphereMask(
   A=ComponentMask(None=MaterialFunctionCall() {materialFunction=&BoundingBoxBased_0-1_UVW}.RGB) {b=true, r=true}
   B=Constant() {r=0.5}
   Radius=ScalarParameter() {defaultValue=0.412, parameterName=SphereRadius}
   Hardness=ScalarParameter() {defaultValue=0.8447, parameterName=SphereHardness}
  ) {attenuationRadius=0.5, hardnessPercent=2}
  TextureCoordinate()
 ]
}
